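# M_Dice is an unlit masked material for dice effects on Niagara mesh particles.
# T_Dice_FX channels separately drive a pattern of two multiplied panning noises, an Outline-scaled term and a desaturated layer.
# Particle alpha thresholds static tile noise to produce a dissolve Opacity value and a hard, bright LightColor edge band.
# The summed layers are tinted by particle color and passed through MF_FX_EmissiveAdjustments.
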Material /DealershipSharedLibrary/Assets/Sets/Cosmicdice/M_Dice
graph {
 properties {
  bAutomaticallySetUsageInEditor=false
  bUsedWithNiagaraMeshParticles=true
  blendMode=BLEND_Masked
  shadingModel=MSM_Unlit
  usedShadingModels=Unlit
 }
 nodes [
  n1 = TextureSample() {texture=/DealershipSharedLibrary/Assets/Sets/Dices/Textures/T_Dice_FX}
  n2 = TextureCoordinate()
  n3 = TextureCoordinate() {uTiling=0.5, vTiling=0.5}
  n4 = Multiply(
   A=TextureSample(UVs=Panner(Coordinate=@n2) {speedX=0.8}) {texture=/Game/Effects/FogVolume/T_TileNoise}.RGB
   B=TextureSample(UVs=Panner(Coordinate=@n3) {speedX=0.3}) {texture=&T_TileNoise}.RGB
  )
  n5 = ScalarParameter() {defaultValue=0.002, parameterName=Outline}
  n6 = Add(A=Add(A=Multiply(A=@n1.R, B=@n4), B=Multiply(A=@n1.B, B=@n5) {constB=0.002}), B=Desaturation(None=@n1.G, Fraction=Constant()))
  n7 = TextureSample(UVs=TextureCoordinate()) {texture=&T_TileNoise}
  n8 = LinearInterpolate(Alpha=ParticleColor().A) {constB=1.5}
  n9 = Subtract(A=Add(A=@n7.G) {constB=0.35}, B=@n8)
  n10 = Multiply(A=Multiply(A=VectorParameter() {parameterName=LightColor, defaultValue=<rgba 1 1 1 1>}.RGB) {constB=32}, B=Saturate(None=Ceil(None=@n9)))
  MaterialFunctionCall(Emissive=Multiply(A=ParticleColor().RGB, B=Add(A=@n6, B=@n10))) {materialFunction=&MF_FX_EmissiveAdjustments}
  MaterialExpressionConstantClamp_0 = Saturate(None=Subtract(A=@n8, B=@n7.G))
  NamedRerouteDeclaration(None=@MaterialExpressionConstantClamp_0) {name=Opacity, nodeColor=<rgba 1 0.375 0 1>}
  NamedRerouteUsage()
 ]
}
